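; Backpack Blueprint for the ScallopLava cosmetic that switches its idle VFX off while the matching ScallopLava pickaxe is the current weapon.
; On weapon change it stores the new weapon, then compares that weapon's skeletal mesh to the axe mesh to deactivate or reactivate the idle VFX component.
; The two post-res-in events repeat the same check on the stored weapon after a short delay, each calling its parent first.
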(event EventOnWeaponChanged
  (params (NewWeapon :type Fort%20Weapon%20Object%20Reference) (OldWeapon :type Fort%20Weapon%20Object%20Reference))
  (bind _returnvalue (CallFunction|GetIdleVFXComponent))
  (|CallParentFunction NewWeapon OldWeapon)
  (Variables|Default|SetNew%20Weapon NewWeapon)
  (if (== (Variables|/Script/Engine.SkinnedMeshComponent|GetSkeletalMesh (Variables|/Script/FortniteGame.FortWeapon|GetWeaponMesh (Variables|Default|GetNew%20Weapon))) "/BRCosmetics/Weapons/FORT_Melee/Pickaxe_ScallopLava/Meshes/ScallopLava_Axe.ScallopLava_Axe")
    (Components|Activation|Deactivate _returnvalue)
    (else
      (Components|Activation|Activate _returnvalue true))))

(event EventOnCosmeticPartAppliedPostResIn
  (params (Part :type Custom%20Character%20Part%20Object%20Reference))
  (bind _returnvalue (CallFunction|GetIdleVFXComponent))
  (|CallParentFunction Part)
  (Utilities|FlowControl|Delay 0.2)
  (if (== (Variables|/Script/Engine.SkinnedMeshComponent|GetSkeletalMesh (Variables|/Script/FortniteGame.FortWeapon|GetWeaponMesh (Variables|Default|GetNew%20Weapon))) "/BRCosmetics/Weapons/FORT_Melee/Pickaxe_ScallopLava/Meshes/ScallopLava_Axe.ScallopLava_Axe")
    (Components|Activation|Deactivate _returnvalue)
    (else
      (Components|Activation|Activate _returnvalue true))))

(event EventOnPostResIn
  (bind _returnvalue (CallFunction|GetIdleVFXComponent))
  (|CallParentFunction)
  (Utilities|FlowControl|Delay 0.2)
  (if (== (Variables|/Script/Engine.SkinnedMeshComponent|GetSkeletalMesh (Variables|/Script/FortniteGame.FortWeapon|GetWeaponMesh (Variables|Default|GetNew%20Weapon))) "/BRCosmetics/Weapons/FORT_Melee/Pickaxe_ScallopLava/Meshes/ScallopLava_Axe.ScallopLava_Axe")
    (Components|Activation|Deactivate _returnvalue)
    (else
      (Components|Activation|Activate _returnvalue true))))

(fn ConstructionScript
  (|CallParentFunction))
