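# Niagara idle effect for the ScallopLava backpack that emits small orange-red sparks from two emitters.
# SparksLocal spawns sparse, longer-lived sparks and SparksWold denser, short-lived ones inheriting some velocity, both using curl noise and shrinking sprites.
# System-level InterpolateOverTime averages gate each emitter's spawn probability and alpha; a bisSwinging flag and vector-length comparison inputs exist, but their wiring is not shown.
# Each emitter has two sprite renderers and writes an additional module-namespace color and size set.
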NiagaraSystem /BRCosmetics/Accessories/FORT_Backpacks/Backpack_M_MED_ScallopLava/FX/NS_Backpack_ScallopLavaIdle
dependencies {
 usedDataInterfaces [&NiagaraDataInterfaceCurve]
 usedDynamicInputs [
  &MakeCustomFloatFromBool
  &SetBoolByFloatComparison
  &VectorLength
 ]
 usedModules [&SystemState, &InterpolateOverTime]
}
emitters {
 SparksLocal {
  emitterUpdateScript [
   &EmitterState()
   &SpawnRate(
    Spawn Probability=&Multiply_Float(A=$System.InterpolateOverTime.InterpolateOverTime_Float.Moving Average, B=1)
    SpawnRate=7
    Use Spawn Probability=true
   )
  ]
  particleSpawnScript [
   &InitializeParticle(
    Color=<rgba 30 2.1362 0 1>
    Lifetime Max=1.2
    Lifetime Min=0.7
    Lifetime Mode=::NewEnumerator1
    Position Offset=&RandomRangeVector(Minimum=<0 0 0>, Maximum=<0 0 0>)
    Sprite Size Max=<0.2 1>
    Sprite Size Min=<0.2 1>
    Sprite Size Mode=::NewEnumerator2
    UsePositionOffset=false
    Write Material Random=true
   )
   &AddVelocity(Velocity=<0 0 30>)
   &ShapeLocation(
    Box Midpoint=<0.5 0 0.5>
    Box Size=<10 20 1>
    MaskSpawnGroup=false
    Offset Mode=::NewEnumerator0
    Override Seed=false
    Shape Primitive=::NewEnumerator2
    Surface Only=false
    UseOwnerScale=true
   )
  ]
  particleUpdateScript [
   /SetVariables(Particles.SizeScale:NiagaraFloat=&FloatFromCurve(
     FloatCurve=DataInterface() {propertyValues {Curve {keys [<at 0 1 RCIM_Cubic>, {arriveTangent=-2.2056, interpMode=RCIM_Cubic, leaveTangent=-2.2056, tangentMode=RCTM_User, time=1, value=0}]}}}
     Scale Curve=&Multiply_Float(A=0.7, B=1)
    )) {isSetParameters=true, module=SetVariables_#1}
   &ParticleState()
   &ScaleSpriteSize(
    Uniform Curve Scale=$Particles.SizeScale
    Uniform Curve Sprite Scale=DataInterface() {propertyValues {Curve {
       defaultValue=0
       keys [
        {interpMode=RCIM_Cubic, tangentMode=RCTM_User, time=0, value=1}
        {arriveTangent=-3.0769, interpMode=RCIM_Cubic, leaveTangent=-3.0769, tangentMode=RCTM_User, time=1, value=0}
       ]
      }}}
   )
   &Drag(Drag=3, Use Linear Drag=true, Use Rotational Drag=true)
   &CurlNoiseForce(
    Bias Noise=false
    Curl Noise Cone Mask Angle=200
    Curl Noise Cone Mask Falloff Angle=30
    Mask Curl Noise=true
    Noise Strength=100
    Pan Noise=false
    Random Seed=15
   )
   &InheritSourceMovement(Applied Position Scale=<0.2 0.2 0.2>, ApplyForce=false, ApplyPosition=true, ApplyPositionFalloff=false)
   &SolveForcesAndVelocity(Clamp Velocity=false, Limit Acceleration=false)
   &ScaleColor(
    Scale Alpha=&Multiply_Float(
     A=&Multiply_Float(
      A=$System.InterpolateOverTime.InterpolateOverTime_Float.Moving Average
      B=$System.InterpolateOverTime002.InterpolateOverTime_Float.Moving Average
     )
     B=1
    )
    ScaleA=true
    ScaleRGB=true
   )
   &FN_Sprite_Attributes(
    Color=&MakeLinearColorFromVectorAndFloat(
     "Vector (RGB)"=&Multiply_Vector(A=&MakeVectorFromLinearColorRGB(Color=<rgba 1 0.1191 0.0887 1>), B=<1 1 1>)
     "Float (Alpha)"=&Multiply_Float(
      A=0.5
      B=&Multiply_Float(
       A=$System.InterpolateOverTime.InterpolateOverTime_Float.Moving Average
       B=$System.InterpolateOverTime002.InterpolateOverTime_Float.Moving Average
      )
     )
    )
    Create Color Attribute=true
    Create SpriteSize Attribute=true
    SpriteSize=&Multiply_Vector2D(A=<3 5>, B=&Multiply_Vector2D_ByFloat(Float=$Particles.SizeScale, Vector2D=<1 1>))
    Write to Module Namespace=true
   )
  ]
  rendererClasses [&NiagaraSpriteRendererProperties]
  renderers [
   {rendererClass=&NiagaraSpriteRendererProperties}
   {rendererClass=&NiagaraSpriteRendererProperties, rendererIndex=1}
  ]
 }
 SparksWold {
  emitterUpdateScript [
   &EmitterState()
   &SpawnRate(
    Spawn Probability=&Multiply_Float(
     A=$System.InterpolateOverTime001.InterpolateOverTime_Float.Moving Average
     B=$System.InterpolateOverTime002.InterpolateOverTime_Float.Moving Average
    )
    SpawnRate=25
    Use Spawn Probability=true
   )
  ]
  particleSpawnScript [
   &InitializeParticle(
    Color=<rgba 20 1.4241 0 1>
    Lifetime Max=0.2
    Lifetime Min=0.3
    Lifetime Mode=::NewEnumerator1
    Position Offset=&RandomRangeVector(Minimum=<0 0 0>, Maximum=<0 0 0>)
    Sprite Size Max=<0.3 1>
    Sprite Size Min=<0.2 1>
    Sprite Size Mode=::NewEnumerator2
    UsePositionOffset=false
    Write Material Random=true
   )
   &AddVelocity(Velocity=<0 0 10>)
   &ShapeLocation(
    Box Midpoint=<0.5 0 0.5>
    Box Size=<10 20 10>
    MaskSpawnGroup=false
    Offset=<-10 0 0>
    Offset Coordinate Space=::Local
    Offset Mode=::NewEnumerator0
    Override Seed=false
    Shape Primitive=::NewEnumerator2
    Surface Only=false
    UseOwnerScale=true
   )
   &InheritVelocity(Inherited Velocity Amount Scale=<0.1 0.1 0.1>, Speed Limit=true, UseSpeedThreshold=false)
  ]
  particleUpdateScript [
   /SetVariables(Particles.SizeScale:NiagaraFloat=&FloatFromCurve(
     FloatCurve=DataInterface() {propertyValues {Curve {keys [<at 0 1 RCIM_Cubic>, {arriveTangent=-2.2056, interpMode=RCIM_Cubic, leaveTangent=-2.2056, tangentMode=RCTM_User, time=1, value=0}]}}}
     Scale Curve=1.3
    )) {isSetParameters=true, module=SetVariables_#1}
   &ParticleState()
   &ScaleSpriteSize(Uniform Curve Sprite Scale=DataInterface() {propertyValues {Curve {
       defaultValue=0
       keys [
        {interpMode=RCIM_Cubic, tangentMode=RCTM_User, time=0, value=1}
        {arriveTangent=-3.0769, interpMode=RCIM_Cubic, leaveTangent=-3.0769, tangentMode=RCTM_User, time=1, value=0}
       ]
      }}})
   &Drag(Use Linear Drag=true, Use Rotational Drag=true)
   &CurlNoiseForce(
    Bias Noise=false
    Curl Noise Cone Mask Angle=200
    Curl Noise Cone Mask Falloff Angle=30
    Mask Curl Noise=true
    Noise Strength=150
    Pan Noise=false
   )
   &InheritSourceMovement(Applied Position Scale=<0.25 0.25 0.4>, ApplyForce=false, ApplyPosition=true, ApplyPositionFalloff=false)
   &SolveForcesAndVelocity(Clamp Velocity=false, Limit Acceleration=false)
   &FN_Sprite_Attributes(
    Color=&MakeLinearColorFromVectorAndFloat(
     "Vector (RGB)"=&Multiply_Vector(A=&MakeVectorFromLinearColorRGB(Color=<rgba 1 0.1191 0.0887 1>), B=<1 1 1>)
     "Float (Alpha)"=&Multiply_Float(
      A=0.5
      B=&Multiply_Float(
       A=$System.InterpolateOverTime001.InterpolateOverTime_Float.Moving Average
       B=$System.InterpolateOverTime002.InterpolateOverTime_Float.Moving Average
      )
     )
    )
    Create Color Attribute=true
    Create SpriteSize Attribute=true
    SpriteSize=&Multiply_Vector2D(A=<3 5>, B=&Multiply_Vector2D_ByFloat(Float=$Particles.SizeScale, Vector2D=<0.8 0.8>))
    Write to Module Namespace=true
   )
   &ScaleColor(
    Scale Alpha=&Multiply_Float(
     A=$System.InterpolateOverTime001.InterpolateOverTime_Float.Moving Average
     B=$System.InterpolateOverTime002.InterpolateOverTime_Float.Moving Average
    )
    ScaleA=true
    ScaleRGB=true
   )
  ]
  rendererClasses [&NiagaraSpriteRendererProperties]
  renderers [
   {rendererClass=&NiagaraSpriteRendererProperties}
   {rendererClass=&NiagaraSpriteRendererProperties, rendererIndex=1}
  ]
 }
}
summary {
 systemName=NS_Backpack_ScallopLavaIdle
 emitters [
  {emitterName=SparksLocal, rendererClasses [&NiagaraSpriteRendererProperties]}
  {emitterName=SparksWold, rendererClasses [&NiagaraSpriteRendererProperties]}
 ]
 userVariables [User.bisSwinging {type=NiagaraBool}]
}
