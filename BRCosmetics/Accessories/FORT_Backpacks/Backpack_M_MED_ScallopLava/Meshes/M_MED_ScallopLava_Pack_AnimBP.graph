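# Animation Blueprint for the ScallopLava backpack cosmetic, simulating its pickaxe and ribbon bones with chained Anim Dynamics.
# A cached mesh pose, layered with body upper-body bones, feeds separate chains; skydiving selects one with a spine-curvature-driven root offset.
# Skydive states add pickaxe and pack offsets, and the event graph updates gravity and curvature alpha.
# The final bool blend has no condition input, apparently leaving the simpler first chain unused.
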AnimBlueprint /BRCosmetics/Accessories/FORT_Backpacks/Backpack_M_MED_ScallopLava/Meshes/M_MED_ScallopLava_Pack_AnimBP
target_skeleton=&/M_MED_ScallopLava_Pack_Skeleton
graphs [
 AnimGraph {nodes [
   Input Pose()
   n1 = Anim Dynamics - Bone: dyn_pickaxe_01!(ComponentPose=Local To Component(LocalPose=Use cached pose 'InputPose'()), Alpha=1, GravityOverride:Vector="0,0,-980", ExternalForce="0,0,0") -> (Pose) {
    angularDampingOverride=0.988
    angularSpringConstant=2550
    bOverrideAngularDamping=true
    bUseGravityOverride=true
    gravityScale=0.35
    alphaBoolBlend {blendOption=Linear}
    boundBone {boneName=dyn_pickaxe_01}
    componentAppliedLinearAccClamp=<200 200 200>
    componentLinearAccScale=<0.4 0.4 0.4>
    componentLinearVelScale=<0.5 0.5 0.5>
    gravityOverride {z=-980}
    physicsBodyDefinitions [{
      boundBone {boneName=dyn_pickaxe_01}
      boxExtents=<10 10 10>
      constraintSetup {angularLimitsMax=<15 4 4>, angularLimitsMin=<-15 -5 -4>, angularTarget=<0 0 0>}
      localJointOffset=<-1 0 0>
     }]
   }
   n2 = Get Gravity Override()
   n3 = Anim Dynamics - Bone: dyn_ribbon_03!(ComponentPose=@n1, Alpha=1, GravityOverride:Vector=@n2, ExternalForce="0,0,0") -> (Pose) {
    angularDampingOverride=0.99
    angularSpringConstant=2550
    bOverrideAngularDamping=true
    bUseGravityOverride=true
    gravityScale=0.35
    alphaBoolBlend {blendOption=Linear}
    boundBone {boneName=dyn_ribbon_03}
    componentAppliedLinearAccClamp=<1000 1000 1000>
    componentLinearAccScale=<0.3 0.3 0.3>
    componentLinearVelScale=<0.3 0.3 0.3>
    gravityOverride {z=-980}
    physicsBodyDefinitions [{
      boundBone {boneName=dyn_ribbon_03}
      boxExtents=<10 10 10>
      constraintSetup {angularLimitsMax=<20 30 20>, angularLimitsMin=<-50 -30 -30>, angularTarget=<0 0 0>}
      localJointOffset=<1 0 0>
     }]
   }
   n4 = Anim Dynamics - Bone: dyn_ribbon_04!(ComponentPose=@n3, Alpha=1, GravityOverride:Vector=@n2, ExternalForce="0,0,0") -> (Pose) {
    angularDampingOverride=0.99
    angularSpringConstant=2550
    bOverrideAngularDamping=true
    bUseGravityOverride=true
    gravityScale=0.35
    alphaBoolBlend {blendOption=Linear}
    boundBone {boneName=dyn_ribbon_04}
    componentAppliedLinearAccClamp=<1000 1000 1000>
    componentLinearAccScale=<0.3 0.3 0.3>
    componentLinearVelScale=<0.3 0.3 0.3>
    gravityOverride {z=-980}
    physicsBodyDefinitions [{
      boundBone {boneName=dyn_ribbon_04}
      boxExtents=<10 10 10>
      constraintSetup {angularLimitsMax=<50 10 20>, angularLimitsMin=<-50 -20 -30>, angularTarget=<0 0 0>}
      localJointOffset=<1.2 0 0>
     }]
   }
   n5 = Anim Dynamics - Bone: dyn_ribbon_05!(ComponentPose=@n4, Alpha=1, GravityOverride:Vector=@n2, ExternalForce="0,0,0") -> (Pose) {
    angularDampingOverride=0.99
    angularSpringConstant=2550
    bOverrideAngularDamping=true
    bUseGravityOverride=true
    gravityScale=0.35
    alphaBoolBlend {blendOption=Linear}
    boundBone {boneName=dyn_ribbon_05}
    componentAppliedLinearAccClamp=<1000 1000 1000>
    componentLinearAccScale=<0.3 0.3 0.3>
    componentLinearVelScale=<0.3 0.3 0.3>
    gravityOverride {z=-980}
    physicsBodyDefinitions [{
      boundBone {boneName=dyn_ribbon_05}
      boxExtents=<10 10 10>
      constraintSetup {angularLimitsMax=<20 20 20>, angularLimitsMin=<-20 -20 -20>, angularTarget=<0 0 0>}
      localJointOffset=<1.2 0 0>
     }]
   }
   n6 = Anim Dynamics - Bone: dyn_ribbon_08!(ComponentPose=@n5, Alpha=1, GravityOverride:Vector=@n2, ExternalForce="0,0,0") -> (Pose) {
    angularDampingOverride=0.99
    angularSpringConstant=2550
    bOverrideAngularDamping=true
    bUseGravityOverride=true
    bUsePlanarLimit=false
    gravityScale=0.35
    alphaBoolBlend {blendOption=Linear}
    boundBone {boneName=dyn_ribbon_08}
    componentAppliedLinearAccClamp=<1000 1000 1000>
    componentLinearAccScale=<0.3 0.3 0.3>
    componentLinearVelScale=<0.3 0.3 0.3>
    gravityOverride {z=-980}
    physicsBodyDefinitions [{
      boundBone {boneName=dyn_ribbon_08}
      boxExtents=<10 10 10>
      constraintSetup {angularLimitsMax=<25 30 20>, angularLimitsMin=<-25 -20 -20>, angularTarget=<0 0 0>}
      localJointOffset=<1 0 0>
     }]
   }
   n7 = Anim Dynamics - Bone: dyn_ribbon_09!(ComponentPose=@n6, Alpha=1, GravityOverride:Vector=@n2, ExternalForce="0,0,0") -> (Pose) {
    angularDampingOverride=0.99
    angularSpringConstant=2550
    bOverrideAngularDamping=true
    bUseGravityOverride=true
    gravityScale=0.35
    alphaBoolBlend {blendOption=Linear}
    boundBone {boneName=dyn_ribbon_09}
    componentAppliedLinearAccClamp=<1000 1000 1000>
    componentLinearAccScale=<0.3 0.3 0.3>
    componentLinearVelScale=<0.3 0.3 0.3>
    gravityOverride {z=-980}
    physicsBodyDefinitions [{
      boundBone {boneName=dyn_ribbon_09}
      boxExtents=<10 10 10>
      constraintSetup {angularLimitsMax=<20 60 20>, angularLimitsMin=<-20 -20 -60>, angularTarget=<0 0 0>}
      localJointOffset=<1.2 0 0>
     }]
   }
   n8 = Get Gravity Override()
   n9 = Anim Dynamics - Bone: dyn_ribbon_06!(ComponentPose=Local To Component(LocalPose=Use cached pose 'InputPose'()), Alpha=1, GravityOverride:Vector=@n8, ExternalForce="0,0,0") -> (Pose) {
    angularBiasOverride=1
    angularDampingOverride=1
    angularSpringConstant=2550
    bOverrideAngularBias=true
    bOverrideAngularDamping=true
    bUseGravityOverride=true
    gravityScale=0.35
    alphaBoolBlend {blendOption=Linear}
    boundBone {boneName=dyn_ribbon_06}
    componentAppliedLinearAccClamp=<500 500 500>
    componentLinearAccScale=<0.3 0.3 0.3>
    componentLinearVelScale=<0.3 0.3 0.3>
    gravityOverride {z=-980}
    physicsBodyDefinitions [{
      boundBone {boneName=dyn_ribbon_06}
      boxExtents=<15 15 15>
      constraintSetup {angularLimitsMax=<0 8 0>, angularLimitsMin=<0 -30 -5>, angularTarget=<0 0 0>}
      localJointOffset=<1 0 0>
     }]
   }
   n10 = Anim Dynamics - Bone: dyn_ribbon_03(ComponentPose=@n9, Alpha=1, GravityOverride=@n8, ExternalForce="0,0,0") {
    angularDampingOverride=0.98
    angularSpringConstant=2550
    bOverrideAngularDamping=true
    bUseGravityOverride=true
    gravityScale=0.35
    alphaBoolBlend {blendOption=Linear}
    boundBone {boneName=dyn_ribbon_03}
    componentAppliedLinearAccClamp=<1000 1000 1000>
    componentLinearAccScale=<0.5 0.5 0.5>
    componentLinearVelScale=<0.5 0.5 0.5>
    gravityOverride {z=-980}
    physicsBodyDefinitions [{
      boundBone {boneName=dyn_ribbon_03}
      boxExtents=<9 9 9>
      constraintSetup {angularLimitsMax=<15 45 20>, angularLimitsMin=<-15 -20 -20>, angularTarget=<0 0 0>}
      localJointOffset=<1 0 0>
     }]
   }
   n11 = Anim Dynamics - Bone: dyn_ribbon_04(ComponentPose=@n10, Alpha=1, GravityOverride=@n8, ExternalForce="0,0,0") {
    angularDampingOverride=0.98
    angularSpringConstant=2550
    bOverrideAngularDamping=true
    bUseGravityOverride=true
    gravityScale=0.35
    alphaBoolBlend {blendOption=Linear}
    boundBone {boneName=dyn_ribbon_04}
    componentAppliedLinearAccClamp=<1000 1000 1000>
    componentLinearAccScale=<0.5 0.5 0.5>
    componentLinearVelScale=<0.5 0.5 0.5>
    gravityOverride {z=-980}
    physicsBodyDefinitions [{
      boundBone {boneName=dyn_ribbon_04}
      boxExtents=<8 8 8>
      constraintSetup {angularLimitsMax=<0 35 10>, angularLimitsMin=<0 -10 -10>, angularTarget=<0 0 0>}
      localJointOffset=<1 0 0>
     }]
   }
   n12 = Anim Dynamics - Bone: dyn_ribbon_05(ComponentPose=@n11, Alpha=1, GravityOverride=@n8, ExternalForce="0,0,0") {
    angularDampingOverride=0.98
    angularSpringConstant=2550
    bOverrideAngularDamping=true
    bUseGravityOverride=true
    gravityScale=0.35
    alphaBoolBlend {blendOption=Linear}
    boundBone {boneName=dyn_ribbon_05}
    componentAppliedLinearAccClamp=<1000 1000 1000>
    componentLinearAccScale=<0.5 0.5 0.5>
    componentLinearVelScale=<0.5 0.5 0.5>
    gravityOverride {z=-980}
    physicsBodyDefinitions [{
      boundBone {boneName=dyn_ribbon_05}
      boxExtents=<7 7 7>
      constraintSetup {angularLimitsMax=<0 35 0>, angularLimitsMin=<0 -10 0>, angularTarget=<0 0 0>}
      localJointOffset=<1 0 0>
     }]
   }
   n13 = Anim Dynamics - Bone: dyn_ribbon_08(ComponentPose=@n12, Alpha=1, GravityOverride=@n8, ExternalForce="0,0,0") {
    angularDampingOverride=0.98
    angularSpringConstant=2550
    bOverrideAngularDamping=true
    bUseGravityOverride=true
    gravityScale=0.35
    alphaBoolBlend {blendOption=Linear}
    boundBone {boneName=dyn_ribbon_08}
    componentAppliedLinearAccClamp=<1000 1000 1000>
    componentLinearAccScale=<0.5 0.5 0.5>
    componentLinearVelScale=<0.5 0.5 0.5>
    gravityOverride {z=-980}
    physicsBodyDefinitions [{
      boundBone {boneName=dyn_ribbon_08}
      boxExtents=<9 9 9>
      constraintSetup {angularLimitsMax=<20 50 20>, angularLimitsMin=<-20 10 -40>, angularTarget=<0 0 0>}
      localJointOffset=<1 0 0>
     }]
   }
   n14 = Anim Dynamics - Bone: dyn_ribbon_09(ComponentPose=@n13, Alpha=1, GravityOverride=@n8, ExternalForce="0,0,0") {
    angularDampingOverride=0.98
    angularSpringConstant=2550
    bOverrideAngularDamping=true
    bUseGravityOverride=true
    gravityScale=0.35
    alphaBoolBlend {blendOption=Linear}
    boundBone {boneName=dyn_ribbon_09}
    componentAppliedLinearAccClamp=<1000 1000 1000>
    componentLinearAccScale=<0.5 0.5 0.5>
    componentLinearVelScale=<0.5 0.5 0.5>
    gravityOverride {z=-980}
    physicsBodyDefinitions [{
      boundBone {boneName=dyn_ribbon_09}
      boxExtents=<6 6 6>
      constraintSetup {angularLimitsMax=<20 20 20>, angularLimitsMin=<-20 -20 -20>, angularTarget=<0 0 0>}
      localJointOffset=<1 0 0>
     }]
   }
   n15 = "Transform (Modify) Bone - Bone: root"(ComponentPose=@n14, Alpha=Get SpineCurvatureAlpha() -> (SpineCurvatureAlpha:double), Translation="0,-1,0", Rotation="0,0,20") {rotationMode=BMM_Additive, translationMode=BMM_Additive, boneToModify {boneName=root}, rotation {roll=20}, translation {y=-1}}
   n16 = Component To Local(ComponentPose=@n15)
   n17 = Anim Dynamics - Bone: dyn_ribbon_06(ComponentPose=Local To Component(LocalPose=Use cached pose 'InputPose'()), Alpha=1, GravityOverride=@n8, ExternalForce="0,0,0") {
    angularBiasOverride=0.9
    angularDampingOverride=0.9
    angularSpringConstant=2550
    bOverrideAngularBias=true
    bOverrideAngularDamping=true
    bUseGravityOverride=true
    gravityScale=0.35
    alphaBoolBlend {blendOption=Linear}
    boundBone {boneName=dyn_ribbon_06}
    componentAppliedLinearAccClamp=<500 500 500>
    componentLinearAccScale=<0.3 0.3 0.3>
    componentLinearVelScale=<0.3 0.3 0.3>
    gravityOverride {z=-980}
    physicsBodyDefinitions [{
      boundBone {boneName=dyn_ribbon_06}
      boxExtents=<15 15 15>
      constraintSetup {angularLimitsMax=<0 8 0>, angularLimitsMin=<0 -40 -10>, angularTarget=<0 0 0>}
      localJointOffset=<1.5 0 0>
     }]
   }
   n18 = Anim Dynamics - Bone: dyn_pickaxe_01(ComponentPose=@n17, Alpha=1, GravityOverride="0,0,-500", ExternalForce="0,0,0") {
    angularDampingOverride=0.988
    angularSpringConstant=2550
    bOverrideAngularDamping=true
    bUseGravityOverride=true
    gravityScale=0.35
    alphaBoolBlend {blendOption=Linear}
    boundBone {boneName=dyn_pickaxe_01}
    componentAppliedLinearAccClamp=<200 200 200>
    componentLinearAccScale=<0.4 0.4 0.4>
    componentLinearVelScale=<0.5 0.5 0.5>
    gravityOverride {z=-980}
    physicsBodyDefinitions [{
      boundBone {boneName=dyn_pickaxe_01}
      boxExtents=<10 10 10>
      constraintSetup {angularLimitsMax=<15 4 4>, angularLimitsMin=<-15 -5 -4>, angularTarget=<0 0 0>}
      localJointOffset=<-1 0 0>
     }]
   }
   n19 = Anim Dynamics - Bone: dyn_ribbon_03(ComponentPose=@n18, Alpha=1, GravityOverride=@n8, ExternalForce="0,0,0") {
    angularDampingOverride=0.98
    angularSpringConstant=2550
    bOverrideAngularDamping=true
    bUseGravityOverride=true
    gravityScale=0.35
    alphaBoolBlend {blendOption=Linear}
    boundBone {boneName=dyn_ribbon_03}
    componentAppliedLinearAccClamp=<1000 1000 1000>
    componentLinearAccScale=<0.5 0.5 0.5>
    componentLinearVelScale=<0.5 0.5 0.5>
    gravityOverride {z=-980}
    physicsBodyDefinitions [{
      boundBone {boneName=dyn_ribbon_03}
      boxExtents=<9 9 9>
      constraintSetup {angularLimitsMax=<20 30 20>, angularLimitsMin=<-50 -30 -30>, angularTarget=<0 0 0>}
      localJointOffset=<1 0 0>
     }]
   }
   n20 = Anim Dynamics - Bone: dyn_ribbon_04(ComponentPose=@n19, Alpha=1, GravityOverride=@n8, ExternalForce="0,0,0") {
    angularDampingOverride=0.98
    angularSpringConstant=2550
    bOverrideAngularDamping=true
    bUseGravityOverride=true
    gravityScale=0.35
    alphaBoolBlend {blendOption=Linear}
    boundBone {boneName=dyn_ribbon_04}
    componentAppliedLinearAccClamp=<1000 1000 1000>
    componentLinearAccScale=<0.5 0.5 0.5>
    componentLinearVelScale=<0.5 0.5 0.5>
    gravityOverride {z=-980}
    physicsBodyDefinitions [{
      boundBone {boneName=dyn_ribbon_04}
      boxExtents=<8 8 8>
      constraintSetup {angularLimitsMax=<50 10 20>, angularLimitsMin=<-50 -20 -30>, angularTarget=<0 0 0>}
      localJointOffset=<1 0 0>
     }]
   }
   n21 = Anim Dynamics - Bone: dyn_ribbon_05(ComponentPose=@n20, Alpha=1, GravityOverride=@n8, ExternalForce="0,0,0") {
    angularDampingOverride=0.98
    angularSpringConstant=2550
    bOverrideAngularDamping=true
    bUseGravityOverride=true
    gravityScale=0.35
    alphaBoolBlend {blendOption=Linear}
    boundBone {boneName=dyn_ribbon_05}
    componentAppliedLinearAccClamp=<1000 1000 1000>
    componentLinearAccScale=<0.5 0.5 0.5>
    componentLinearVelScale=<0.5 0.5 0.5>
    gravityOverride {z=-980}
    physicsBodyDefinitions [{
      boundBone {boneName=dyn_ribbon_05}
      boxExtents=<7 7 7>
      constraintSetup {angularLimitsMax=<20 20 20>, angularLimitsMin=<-20 -20 -20>, angularTarget=<0 0 0>}
      localJointOffset=<1 0 0>
     }]
   }
   n22 = Anim Dynamics - Bone: dyn_ribbon_08(ComponentPose=@n21, Alpha=1, GravityOverride=@n8, ExternalForce="0,0,0") {
    angularDampingOverride=0.98
    angularSpringConstant=2550
    bOverrideAngularDamping=true
    bUseGravityOverride=true
    bUsePlanarLimit=false
    gravityScale=0.35
    alphaBoolBlend {blendOption=Linear}
    boundBone {boneName=dyn_ribbon_08}
    componentAppliedLinearAccClamp=<1000 1000 1000>
    componentLinearAccScale=<0.5 0.4 0.4>
    componentLinearVelScale=<0.5 0.4 0.4>
    gravityOverride {z=-980}
    physicsBodyDefinitions [{
      boundBone {boneName=dyn_ribbon_08}
      boxExtents=<15 15 15>
      constraintSetup {angularLimitsMax=<25 30 20>, angularLimitsMin=<-25 -20 -20>, angularTarget=<0 0 0>}
      localJointOffset=<1.1 0 0>
     }]
   }
   n23 = Anim Dynamics - Bone: dyn_ribbon_09(ComponentPose=@n22, Alpha=1, GravityOverride=@n8, ExternalForce="0,0,0") {
    angularDampingOverride=0.9
    angularSpringConstant=2550
    bOverrideAngularDamping=true
    bUseGravityOverride=true
    gravityScale=0.35
    alphaBoolBlend {blendOption=Linear}
    boundBone {boneName=dyn_ribbon_09}
    componentAppliedLinearAccClamp=<1000 1000 1000>
    componentLinearAccScale=<0.5 0.5 0.5>
    componentLinearVelScale=<0.5 0.5 0.5>
    gravityOverride {z=-980}
    physicsBodyDefinitions [{
      boundBone {boneName=dyn_ribbon_09}
      boxExtents=<10 10 10>
      constraintSetup {angularLimitsMax=<20 60 20>, angularLimitsMin=<-20 -20 -60>, angularTarget=<0 0 0>}
      localJointOffset=<1.5 0 0>
     }]
   }
   n24 = Component To Local(ComponentPose=@n23)
   n25 = Get bIsSkyDiving()
   n26 = Anim Dynamics - Bone: dyn_ribbon_07!(
    ComponentPose=Local To Component(LocalPose=Two Way Blend!(A=@n24, B=@n16, bAlphaBoolEnabled:bool=@n25) -> (Pose) {blendNode {alphaInputType=Bool, alphaBoolBlend {blendOption=Linear}}})
    Alpha=1
    GravityOverride:Vector=@n8
    ExternalForce="0,0,0"
   ) -> (Pose) {
    angularDampingOverride=0.98
    angularSpringConstant=2550
    bOverrideAngularDamping=true
    bUseGravityOverride=true
    gravityScale=0.35
    alphaBoolBlend {blendOption=Linear}
    boundBone {boneName=dyn_ribbon_07}
    componentAppliedLinearAccClamp=<1000 1000 1000>
    componentLinearAccScale=<0.5 0.5 0.5>
    componentLinearVelScale=<0.5 0.5 0.5>
    gravityOverride {z=-980}
    physicsBodyDefinitions [{
      boundBone {boneName=dyn_ribbon_07}
      boxExtents=<6 6 6>
      constraintSetup {angularLimitsMax=<0 2 0>, angularLimitsMin=<0 -5 0>, angularTarget=<0 0 0>}
      localJointOffset=<1 0 0>
     }]
   }
   n27 = "Transform (Modify) Bone - Bone: dyn_pickaxe_01"!(ComponentPose=@n26, bAlphaBoolEnabled:bool=Get bIsSkyDiving(), Translation="0,1.5,0", Rotation="8,-2,-20", Scale="1,1,1") -> (Pose) {
    alphaInputType=Bool
    rotationMode=BMM_Additive
    rotationSpace=BCS_BoneSpace
    translationMode=BMM_Additive
    translationSpace=BCS_BoneSpace
    alphaBoolBlend {blendInTime=0.2, blendOption=Linear, blendOutTime=0.2}
    boneToModify {boneName=dyn_pickaxe_01}
    rotation=<rot 8 -2 -20>
    translation {y=1.5}
   }
   n28 = "Transform (Modify) Bone - Bone: pack"!(ComponentPose=@n27, bAlphaBoolEnabled:bool=Get bIsSkydiveFloating(), Translation="0,-2,0", Rotation="0,0,0", Scale) -> (Pose) {
    alphaInputType=Bool
    rotationMode=BMM_Additive
    rotationSpace=BCS_BoneSpace
    translationMode=BMM_Additive
    alphaBoolBlend {blendInTime=0.2, blendOption=Linear, blendOutTime=0.2}
    boneToModify {boneName=pack}
    translation {y=-2}
   }
   Output Pose(Result=Blend Poses by bool(BlendPose_0=Component To Local(ComponentPose=@n7), BlendPose_1=Component To Local(ComponentPose=@n28))) {layerGroup=None}
   n29 = Layered blend per bone(
    BasePose=Copy Pose From Mesh()
    BlendPoses_0=Copy Pose From Mesh(SourceMeshComponent=Get Body Skeletal Mesh() -> (Body Skeletal Mesh:Skeletal Mesh Component Object Reference))
   ) {layerSetup [{branchFilters [
       {blendDepth=3, boneName=upperarm_r}
       {blendDepth=3, boneName=upperarm_l}
       {boneName=clavicle_r}
       {boneName=clavicle_l}
       {boneName=spine_5}
       {boneName=spine_4}
       {boneName=spine_3}
      ]}]}
   Save cached pose 'InputPose'(Pose=@n29) {cacheName=InputPose}
  ]}
 EventGraph {nodes [
   Event BlueprintUpdateAnimation()
   > n1 = Sequence()
   > n2 = Branch(Condition=Get bIsSkyDiving())
   > Set Gravity Override(Gravity Override="0,0,-400")
   Set Gravity Override(execute=@n2.else, Gravity Override="0,0,-700")
   Set SpineCurvatureAlpha(
    execute=@n1.then_1
    SpineCurvatureAlpha:double=GetAlphaFromReferenceTransformDelta(SkeletalMeshComponent=Get Body Skeletal Mesh(), BoneName=spine_05, BaseName=spine_01, InputRange=(X=0,Y=-22), TwistAxis="1,0,0")
   ) -> (Output_Get:double)
   Event OnCharacterCustomizationLoaded()
   > n3 = Cast To FortPlayerPawn(Object=GetOwningActor())
   > Set Body Skeletal Mesh(Body Skeletal Mesh:Skeletal Mesh Component Object Reference=GetSkeletalMeshForPartType(self=@n3)) -> (Output_Get:Skeletal Mesh Component Object Reference)
  ]}
]
variables [
 Body Skeletal Mesh {type=SkeletalMeshComponent}
 Gravity Override {type=Vector, default=<0 0 -700>}
 SpineCurvatureAlpha {type=number}
]
